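# Additive, unlit, two-sided glow material for Niagara mesh particles, apparently for the Ebony pickaxe effects.
# Panning wave and Perlin textures blend to distort UVs, which drive two scrolling, high-power sine bands offset by particle random.
# The combined bands are masked along V and also scale the vertex normal, suggesting mesh displacement.
# Particle color feeds an emissive adjustment function, and particle alpha multiplies a radial gradient.
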Material /BRCosmetics/Weapons/FORT_Melee/Pickaxe_Ebony/FX/M_EbonyGlow
graph {
 properties {
  bUsedWithNiagaraMeshParticles=true
  blendMode=BLEND_Additive
  pixelDepthOffsetMode=PDOM_Legacy
  shadingModel=MSM_Unlit
  twoSided=true
  usedShadingModels=Unlit
 }
 nodes [
  Constant() {r=1}
  n1 = ParticleColor()
  MaterialFunctionCall(Emissive=Multiply(A=@n1.RGB)) {materialFunction=&MF_FX_EmissiveAdjustments}
  n2 = Power(Base=ComponentMask(None=TextureCoordinate()) {g=true}) {constExponent=3}
  n3 = MaterialFunctionCall(Float2=TextureCoordinate()) {materialFunction=&BreakOutFloat2Components}
  n4 = Constant4Vector() {constant=<rgba 3 1 0.3 1>}
  n5 = Multiply(A=@n3.R, B=@n4.R)
  n6 = Multiply(A=@n3.G, B=@n4.G)
  n7 = TextureSampleParameter2D(UVs=Panner(Coordinate=AppendVector(A=@n5, B=@n6), Speed=AppendVector(A=@n4.B, B=@n4.B))) {parameterName=TEx1, samplerType=SAMPLERTYPE_LinearGrayscale, texture=/Game/UI/Frontend/ItemShop/Art/Winterfest/WF_2022/T_UI_WaveTexture}
  n8 = MaterialFunctionCall(Float2=TextureCoordinate()) {materialFunction=&BreakOutFloat2Components}
  n9 = Constant4Vector() {constant=<rgba 4 1.3 0 1>}
  n10 = Multiply(A=@n8.R, B=@n9.R)
  n11 = Multiply(A=@n8.G, B=@n9.G)
  n12 = TextureSampleParameter2D(UVs=Panner(Coordinate=AppendVector(A=@n10, B=@n11), Speed=AppendVector(A=@n9.B, B=@n9.B))) {
   parameterName=Tex2
   samplerType=SAMPLERTYPE_LinearGrayscale
   texture=&T_SimplePerlin_Greyscale
  }
  n13 = LinearInterpolate(A=Multiply(A=@n7.R) {constB=0.3}, B=@n12.R)
  n14 = Add(A=Multiply(A=Power(Base=@n13)) {constB=0.8}, B=TextureCoordinate())
  n15 = Time()
  n16 = ParticleRandom()
  n17 = Multiply(A=@n16) {constB=2}
  n18 = ComponentMask(None=Panner(Coordinate=@n14, Time=Add(A=@n15, B=@n17)) {speedX=0.2, speedY=-1.7}) {g=true}
  n19 = Power(Base=Sine(None=Multiply(A=@n18) {constB=0.5})) {constExponent=200}
  n20 = Add(A=@n15, B=Multiply(A=@n16) {constB=7})
  n21 = Multiply(A=ComponentMask(None=Panner(Coordinate=@n14, Time=@n20) {speedX=-0.5, speedY=-1.2}) {g=true}) {constB=0.5}
  n22 = Saturate(None=Power(Base=Sine(None=@n21)) {constExponent=20})
  n23 = Add(A=ComponentMask(None=Saturate(None=@n19)) {r=true}, B=Multiply(A=ComponentMask(None=@n22) {r=true}) {constB=0.3})
  Multiply(A=Saturate(None=Multiply(A=@n2) {constB=100}), B=@n23)
  n24 = TextureCoordinate()
  n25 = Constant() {r=1}
  Multiply(
   A=@n1.A
   B=Multiply(A=MaterialFunctionCall(UVs=@n24, Density=@n25) {materialFunction=&RadialGradientExponential}.RadialGradientExponential)
  )
  Multiply(A=Multiply(A=@n23, B=VertexNormalWS())) {constB=0.8}
 ]
}
