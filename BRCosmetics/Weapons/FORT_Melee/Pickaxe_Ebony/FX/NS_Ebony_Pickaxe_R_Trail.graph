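# Niagara ribbon trail for the Ebony pickaxe, likely the right-hand one, as suggested by the R in its name.
# Its single FN_AnimTrail emitter spawns and updates short-lived particles with the shared AnimTrail modules, drawn by two ribbon renderers.
# A float comparison on User.Variant picks one of two pink-red tints at spawn.
# Curves flash each particle bright then fade it, while inherited source movement offsets its position.
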NiagaraSystem /BRCosmetics/Weapons/FORT_Melee/Pickaxe_Ebony/FX/NS_Ebony_Pickaxe_R_Trail
dependencies {
 usedDataInterfaces [&NiagaraDataInterfaceColorCurve, &NiagaraDataInterfaceCurve]
 usedDynamicInputs [
  &Multiply_Float
  &FloatFromCurve
  &FNTimeOfDayScalar
 ]
 usedModules [&SystemState]
}
emitters {FN_AnimTrail {
  emitterUpdateScript [
   &EmitterState(
    Inactive Response=::NewEnumerator1
    Life Cycle Mode=::NewEnumerator1
    Loop Behavior=::NewEnumerator1
    Loop Delay=0.05
    Loop Duration=0.3
    UseLoopDelay=false
   )
   &FN_SpawnAnimTrail_v2(Spawn Interval=0)
  ]
  particleSpawnScript [
   /Game/Effects/Niagara/Modules/AnimTrail/FN_InitializeAnimTrail/FN_InitializeAnimTrailBoneIndecies(First Bone or Socket Index=1)
   /Game/Effects/Niagara/Modules/AnimTrail/FN_InitializeAnimTrail/FN_InitializeAnimTrail_v3/FN_InitializeAnimTrail(
    AnimTrailWidth=1.1
    Color:LinearColor=&MakeCustomLinearColorFromBool(
     Bool=&SetBoolByFloatComparison(A=$User.Variant, B=0.5, Comparison Type=::NewEnumerator7)
     True Color=<rgba 1.3 0 0.104 1>
     False Color=<rgba 1 0 0.5 0.1>
    )
    Lifetime=0.26
   )
  ]
  particleUpdateScript [
   /Game/Effects/Niagara/Modules/AnimTrail/FN_UpdateAnimTrail/FN_UpdateAnimTrail_v3/FN_UpdateAnimTrail(Locked Particle RibbonWidth:NiagaraFloat=$Particles.RibbonWidth)
   &ScaleColor(
    Linear Color Curve=DataInterface() {propertyValues {
      AlphaCurve {keys [<at 0 0>,<at 0.0833 1>,<at 0.592 1>,<at 1 0>]}
      BlueCurve {keys [<at 0.0537 8>,<at 0.3521 1>]}
      GreenCurve {keys [<at 0.0537 8>,<at 0.3521 1>]}
      RedCurve {keys [<at 0.0537 8>,<at 0.3521 1>]}
     }}
    Scale Mode=::NewEnumerator2
   )
   &InheritSourceMovement(Applied Position Scale=<-0.2 -0.2 0.2>, ApplyForce=false, ApplyPosition=true, ApplyPositionFalloff=false)
   &SolveForcesAndVelocity(Clamp Velocity=false, Limit Acceleration=false)
  ]
  rendererClasses [&NiagaraRibbonRendererProperties]
  renderers [
   {rendererClass=&NiagaraRibbonRendererProperties}
   {rendererClass=&NiagaraRibbonRendererProperties, rendererIndex=1}
  ]
 }}
summary {
 systemName=NS_Ebony_Pickaxe_R_Trail
 emitters [{emitterName=FN_AnimTrail, rendererClasses [&NiagaraRibbonRendererProperties]}]
 userVariables [
  User.RibbonIDOffset {type=NiagaraInt32}
  User.SkeletalMesh {
   type=NiagaraDataInterfaceSkeletalMesh
   defaultValue {
    struct=NiagaraExt_VariableValue_DataInterface
    value {dataInterface=&self:NiagaraDataInterfaceSkeletalMesh_0}
   }
  }
  User.Variant {type=NiagaraFloat, defaultValue=0, description=""}
 ]
}
