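# Niagara spin effect for the Ebony pickaxe, built from a single sprite emitter with an instantaneous burst.
# User.Variant picks one of two tints at spawn, and User.RotationDirection sets the sign of a fast sprite rotation with locally oriented facing.
# Particles are not killed when their lifetime elapses.
# A system Timeline's normalized time drives a cubic curve that ramps color and alpha up from zero.
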NiagaraSystem /BRCosmetics/Weapons/FORT_Melee/Pickaxe_Ebony/FX/NS_Ebony_Pickaxe_SpinFX
dependencies {
 usedDataInterfaces [&NiagaraDataInterfaceColorCurve, &NiagaraDataInterfaceCurve]
 usedDynamicInputs [
  &MakeVectorFromLinearColorRGB
  &ColorFromCurve
  &MakeFloatFromVector4
 ]
 usedModules [
  &SystemState
  &self:SystemSpawnScript.SystemScriptSource.SystemScriptGraph.NiagaraNodeAssignment_0.SetVariables_#1
  &Timeline
 ]
}
emitters {Spin {
  emitterUpdateScript [
   &EmitterState()
   &SpawnBurst_Instantaneous(Use Spawn Probability=false, UseLoopCountLimit=true)
  ]
  particleSpawnScript [&InitializeParticle(
    Color=&MakeCustomLinearColorFromBool(
     Bool=&SetBoolByFloatComparison(A=$User.Variant, B=0.5, Comparison Type=::NewEnumerator7)
     True Color=<rgba 0.3 0.012 0.0216 0.1>
     False Color=<rgba 0.25 0.03 0.1217 0.1>
    )
    Lifetime=0.4
    Position Offset=<-15 -10 0>
    Sprite Rotation Angle=90
    Sprite Rotation Mode=::NewEnumerator0
    Sprite Size Mode=::NewEnumerator3
    Uniform Sprite Size=110
    UsePositionOffset=false
    Write Material Random=true
   )]
  particleUpdateScript [
   &ParticleState(Kill Particles When Lifetime Has Elapsed=false, Let Infinitely Lived Particles Die When Emitter Deactivates=true)
   &SpriteRotationRate(Rotation Rate=&Multiply_Float(A=&Multiply_Float(A=-1200, B=$User.RotationDirection), B=1))
   &ScaleColor(
    Scale Mode=::NewEnumerator1
    Scale RGBA=&MakeVector4_FromFloat(Float=&FloatFromCurve(
      FloatCurve=DataInterface() {propertyValues {Curve {keys [{interpMode=RCIM_Cubic, tangentMode=RCTM_User, time=0, value=0}, {interpMode=RCIM_Cubic, tangentMode=RCTM_User, time=0.9, value=1}]}}}
      CurveIndex=$System.Timeline.NormalizedTime
     ))
    ScaleRGBA=true
   )
   &SpriteFacingAndAlignment(Facing Coordinate Space=::Local, WriteAlignment=false, WriteFacing=true)
  ]
  rendererClasses [&NiagaraSpriteRendererProperties]
  renderers [{rendererClass=&NiagaraSpriteRendererProperties}]
 }}
summary {
 systemName=NS_Ebony_Pickaxe_SpinFX
 emitters [{emitterName=Spin, rendererClasses [&NiagaraSpriteRendererProperties]}]
 userVariables [
  User.FadeFX {type=NiagaraFloat, defaultValue=0, description=""}
  User.IndexColors {type=NiagaraInt32}
  User.RibbonIDOffset {type=NiagaraInt32}
  User.RotationDirection {type=NiagaraFloat, defaultValue=1, description=""}
  User.SkeletalMesh {
   type=NiagaraDataInterfaceSkeletalMesh
   defaultValue {
    struct=NiagaraExt_VariableValue_DataInterface
    value {dataInterface=&self:NiagaraDataInterfaceSkeletalMesh_0}
   }
  }
  User.TrailColor {type=LinearColor, description="", defaultValue {struct=LinearColor, value=<rgba 1 1 1 0.35>}}
  User.Variant {type=NiagaraFloat, defaultValue=0, description=""}
 ]
}
